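# Conversation database for the Fishstick NPC, providing its greeting and the branches a player can take from it.
# The entry point leads to a speech node that picks contextual lines by requirement and priority: first-ever meeting, match conversation count, or specific outfits the player wears, with a default message as fallback.
# The speech node then links by tag to shared NPC library quests, random services and a separate Fishstick dialogue entry.
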ConversationDatabase /NPCLibrary/NPCs/Fishstick/Fishstick/NPCLibrary_Conversation_Fishstick
entry_points [{tag=Conversation.Entry.Fishstick, destinations [#1]}]
nodes [
 ConversationEntryPointNode_7 {
  id=#1
  class=ConversationEntryPointNode
  outputs [#2]
  properties {entryTag {tagName=Conversation.Entry.Fishstick}}
 }
 Speech_C_4 {
  id=#2
  class=Speech_C
  outputs [#3, #4, #5]
  properties {
   nodeName=""
   generalConfig {
    defaultMessage=Default Conversation
    contextualMessages [
     {
      message=(Fishstick smiles proudly. Sticks has a new franchise!)
      contextRequirements [{requirement {class=FortControllerRequirement_IsFirstEverConversationWithNPC}}]
      priority {curve {rowName=Tandem.Speech.Priority.FirstEverInteraction}}
     }
     {
      message=(Fishstick burbles cheerfully.)
      contextRequirements [{requirement {class=FortControllerRequirement_CheckNPCMatchConversationCount, properties {nPCUniqueIdTag {tagName=AISpawnerData.Type.Tandem.Fishstick}}}}]
      priority {curve {curveTable=null, rowName=None}}
     }
     {
      message=(Fishstick winks with both eyes.)
      contextRequirements [{requirement {class=FortControllerRequirement_CheckNPCMatchConversationCount, properties {nPCUniqueIdTag {tagName=AISpawnerData.Type.Tandem.Fishstick}}}}]
      priority {curve {curveTable=null, rowName=None}}
     }
     {
      message=(Fishsticks offers you a coupon for Sticks.)
      contextRequirements [{requirement {class=FortControllerRequirement_CheckNPCMatchConversationCount, properties {nPCUniqueIdTag {tagName=AISpawnerData.Type.Tandem.Fishstick}}}}]
      priority {curve {curveTable=null, rowName=None}}
     }
     {
      message=(Fishstick gives you a friendly wave.)
      contextRequirements [{requirement {class=FortControllerRequirement_HasCID, properties {allowedCIds [/BRCosmetics/Athena/Items/Cosmetics/Characters/Character_SquibblySquid]}}}]
      priority {curve {rowName=Tandem.Speech.Priority.OutfitReaction}}
     }
     {
      message=(Fishstick bows reverently.)
      contextRequirements [{requirement {
         class=FortControllerRequirement_HasCID
         properties {allowedCIds [/BRCosmetics/Athena/Items/Cosmetics/Characters/CID_872_Athena_Commando_F_TeriyakiFishPrincess]}
        }}]
      priority {curve {rowName=Tandem.Speech.Priority.OutfitReaction}}
     }
     {
      message=(Fishstick thanks you for your service.)
      contextRequirements [{requirement {
         class=FortControllerRequirement_HasCID
         properties {allowedCIds [/BRCosmetics/Athena/Items/Cosmetics/Characters/CID_624_Athena_Commando_M_TeriyakiWarrior]}
        }}]
      priority {curve {rowName=Tandem.Speech.Priority.OutfitReaction}}
     }
    ]
   }
  }
 }
 ConversationLinkNode_2 {id=#5, class=ConversationLinkNode, properties {remoteEntryTag {tagName=Conversation.Entry.Fishstick.Dialogue}}}
 ConversationLinkNode_3 {id=#4, class=ConversationLinkNode, properties {remoteEntryTag {tagName=Conversation.Entry.RandomServices}}}
 ConversationLinkNode_1 {id=#3, class=ConversationLinkNode, properties {remoteEntryTag {tagName=Conversation.Entry.NPCLibrary.Quests}}}
]
